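; BPI_LPTT is a Blueprint Interface for the LowPolyTurntable tech-art tool that declares calls for switching display modes.
; It declares functions for default, ambient occlusion, checker, CRV, IDs, vertex color and wireframe views, with no logic of its own.
; Only SetWireframe takes an input, a boolean that controls whether seams are shown.
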(fn SetAO)

(fn SetChecker)

(fn SetCRV)

(fn SetDefault)

(fn SetIDs)

(fn SetVertexColor)

(fn SetWireframe
  (params (ShowSeams? :type Boolean)))
