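; Interchange pipeline Blueprint for the LowPolyTurntable art tool that controls how normals and tangents are handled on static mesh import.
; The scripted execute event loops over the node container's roots and casts each factory node to a static mesh factory node.
; A switch on the NormalsImportSetting enum turns off recomputation of both normals and tangents, or of normals only, for the two handled options.
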(event Interchange|Pipeline|EventScriptedExecutePipeline
  (params (BaseNodeContainer :type Interchange%20Base%20Node%20Container%20Object%20Reference) (SourceDatas :type Array%20of%20Interchange%20Source%20Data%20Object%20References) (ContentBasePath :type String))
  (bind _returnvalue false)
  (for _array_element (Interchange|NodeContainer|GetRoots BaseNodeContainer)
    (bind _asinterchange_static_mesh_factory_node (Utilities|Casting|CastToInterchangeStaticMeshFactoryNode (Interchange|NodeContainer|GetFactoryNode BaseNodeContainer _array_element)))
    (switch Utilities|FlowControl|Switch|SwitchonEnum_LPTT_NormalsImportSettings (Variables|Default|GetNormalsImportSetting)
      (:NewEnumerator1
        (Interchange|FactoryNode|Mesh|SetCustomRecomputeNormals _asinterchange_static_mesh_factory_node _returnvalue true)
        (Interchange|FactoryNode|Mesh|SetCustomRecomputeTangents _asinterchange_static_mesh_factory_node _returnvalue true))
      (:NewEnumerator2
        (Interchange|FactoryNode|Mesh|SetCustomRecomputeNormals _asinterchange_static_mesh_factory_node _returnvalue true)))))
