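; UI controller actor for a low-poly turntable art tool that creates the data display widget and adds it to the viewport on BeginPlay.
; Wrapper events and functions forward to that widget to toggle texture hiding, set pass text, and switch the texture to UV islands or wireframe.
; The construction script stores the platform user name, current date and engine version (as build) in variables.
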(event EventBeginPlay
  (bind _returnvalue (UserInterface|CreateWidget "/Game/NeverCook/ArtTools/Tech_Art/LowPolyTurntable/Data/UI_LPTT_ShowData.UI_LPTT_ShowData_C"))
  (bind _output_get (Variables|Default|SetUI _returnvalue))
  (UserInterface|Viewport|AddtoViewport _output_get 0))

(event Custom|DisplayUI?
  (params (Display? :type Boolean))
  (bind _ui (Variables|Default|GetUI))
  (if Display?
    (Class|UILPTTShowData|HideTextures _ui true)
    (else
      (Class|UILPTTShowData|HideTextures _ui false))))

(event Collision|EventActorBeginOverlap
  (params (OtherActor :type Actor%20Object%20Reference)))

(event EventTick
  (params (DeltaSeconds :type Float%20%28single-precision%29)))

(fn SetPassText
  (params (Text :type Text))
  (Class|UILPTTShowData|SetPassText (Variables|Default|GetUI) Text))

(fn SwitchToUVIslands
  (Passes|SwitchTexturetoUVIslands (Variables|Default|GetUI)))

(fn SwitchToWireframe
  (Passes|SwitchTexturetoWireFrame (Variables|Default|GetUI)))

(fn ConstructionScript
  ;[1 1 1] UserName
  (Variables|Default|SetuserName (Utilities|Platform|GetPlatformUserName))
  ;[1 1 1] Date
  (Variables|Default|Setdate (Utilities|String|ToString(Text) (Utilities|Text|AsDate(fromUTC) (Math|DateTime|UTCNow) "EDT")))
  ;[1 1 1] Build #
  (Variables|Default|Setbuild (Development|GetEngineVersion)))
