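; BP_RigCapture is a scene capture rig Blueprint in the LowPolyTurntable tech art tool that captures to a render target.
; The construction script assigns the Texture Target variable to its SceneCaptureComponent2D.
; A SetRotation custom event turns a direction vector into a rotation from X and applies it as the CameraPivot's world rotation.
; BeginPlay, Tick and overlap events are empty.
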(event EventBeginPlay)

(event Collision|EventActorBeginOverlap
  (params (OtherActor :type Actor%20Object%20Reference)))

(event EventTick
  (params (DeltaSeconds :type Float%20%28single-precision%29)))

(event Custom|SetRotation
  (params (Rotation :type Vector))
  (Transformation|SetWorldRotation (Variables|Default|GetCameraPivot) (Math|Conversions|RotationFromXVector Rotation) false false))

(fn ConstructionScript
  (Variables|/Script/Engine.SceneCaptureComponent2D|SetTextureTarget (Variables|Default|GetTexture%20Target) (Variables|Default|GetSceneCaptureComponent2D)))
