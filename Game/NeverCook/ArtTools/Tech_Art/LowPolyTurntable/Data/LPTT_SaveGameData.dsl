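; LPTT_SaveGameData is a data holder for the Low Poly Turntable art tool, apparently a save game object, storing its import settings.
; SaveData takes an array of import part structs and an array of texture map directory paths and writes them to the ImportPaths and TextureMapsPath variables.
; LoadData returns both stored arrays.
; Neither function performs disk reads or writes itself.
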(fn LoadData
  (return (Variables|Default|GetImportPaths) (Variables|Default|GetTextureMapsPath)))

(fn SaveData
  (params (ImportPaths :type Array%20of%20LPTT%20Import%20Part%20Struct%20Structures) (TextureMapsPath :type Array%20of%20Directory%20Path%20Structures))
  (Variables|Default|SetImportPaths ImportPaths)
  (Variables|Default|SetTextureMapsPath TextureMapsPath))
