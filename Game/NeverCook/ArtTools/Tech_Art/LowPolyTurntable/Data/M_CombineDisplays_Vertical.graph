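# Post-process material for the low-poly turntable art tool that combines two display render targets into one vertically stacked screen image.
# An If on viewport V switches just below mid-screen: the upper part samples LeftImageRenderTarget, the lower part RightImageRenderTarget.
# Each sample uses viewport UVs shifted vertically by its own constant offset to reposition the image within its band.
# The selected color passes through LinearTosRGB before output.
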Material /Game/NeverCook/ArtTools/Tech_Art/LowPolyTurntable/Data/M_CombineDisplays_Vertical
graph {
 properties {bAutomaticallySetUsageInEditor=false, materialDomain=MD_PostProcess, usedShadingModels=Default Lit}
 nodes [
  n1 = ScreenPosition()
  n2 = Add(
   A=@n1.ViewportUV
   B=MaterialFunctionCall(X=Constant(), Y=Constant() {r=-0.3}) {materialFunction=&MakeFloat2}.Result
  )
  n3 = Constant()
  n4 = Constant() {r=0.2}
  n5 = TextureSample(UVs=Add(
    A=@n1.ViewportUV
    B=MaterialFunctionCall(X=@n3, Y=@n4) {materialFunction=&MakeFloat2}.Result
   )) {samplerType=SAMPLERTYPE_LinearColor, texture=&/LeftImageRenderTarget}
  MaterialFunctionCall!(Linear=If(
    A=ComponentMask(None=@n1.ViewportUV) {g=true}
    "A > B"=TextureSample(UVs=@n2) {samplerType=SAMPLERTYPE_LinearColor, texture=&/RightImageRenderTarget}.RGB
    "A == B"=@n5.RGB
    "A < B"=@n5.RGB
   ) {constB=0.535}) {materialFunction=/Engine/Functions/Engine_MaterialFunctions01/ImageAdjustment/LinearTosRGB}
 ]
}
