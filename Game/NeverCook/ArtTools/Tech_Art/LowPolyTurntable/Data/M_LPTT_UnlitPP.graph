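# M_LPTT_UnlitPP is a post-process material in the LowPolyTurntable tech-art tools, named for an unlit view.
# It uses the Unlit shading model and has automatic usage flagging in the editor turned off.
# The graph holds two SceneTexture nodes, one for stored GBuffer base color and one for PostProcessInput0, with no connections shown between them or to an output.
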Material /Game/NeverCook/ArtTools/Tech_Art/LowPolyTurntable/Data/M_LPTT_UnlitPP
graph {
 properties {bAutomaticallySetUsageInEditor=false, materialDomain=MD_PostProcess, shadingModel=MSM_Unlit, usedShadingModels=Unlit}
 nodes [
  SceneTexture() {sceneTextureId=PPI_StoredBaseColor}
  SceneTexture() {sceneTextureId=PPI_PostProcessInput0}
 ]
}
