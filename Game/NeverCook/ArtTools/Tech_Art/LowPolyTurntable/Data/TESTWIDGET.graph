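# TESTWIDGET is an Editor Utility Widget in the LowPolyTurntable tech-art tool folder that acts as a test harness for spawning an actor.
# A grid panel holds a single TA_Generic01_Button labelled TESTCLICK.
# Its On Clicked event calls SpawnActorFromClass to place a TESTACTORBLANK actor at the world origin.
# The PreConstruct, Construct and Tick events are empty, and the widget has no variables.
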EditorUtilityWidgetBlueprint /Game/NeverCook/ArtTools/Tech_Art/LowPolyTurntable/Data/TESTWIDGET
graphs [EventGraph {nodes [
   "On Clicked (TA_Generic01_Button)"!() -> (OutputDelegate, then, widget:User Widget Object Reference, event:Pointer Event Structure) {is=title K2Node_ComponentBoundEvent}
   > SpawnActorFromClass!(
    execute
    ActorClass:Actor Class Reference=&/TESTACTORBLANK.TESTACTORBLANK_C
    Location:Vector="0, 0, 0"
    Rotation:Rotator
    bTransient:bool
   ) -> (then, ReturnValue:TESTACTORBLANK Object Reference)
   Event PreConstruct()
   Event Construct()
   Event Tick()
  ]}]
variables []
widgets {
 info {rootWidgetClass=&GridPanel, widgetCount=2}
 widgets [
  GridPanel_0 {widgetClassPath=&GridPanel, widget=GridPanel_0}
  TA_Generic01_Button {
   widgetClassPath=&TA_Generic01_Button
   parent=GridPanel_0
   widget=TA_Generic01_Button
   properties {label=TESTCLICK}
   slot {class=GridSlot}
  }
 ]
}
