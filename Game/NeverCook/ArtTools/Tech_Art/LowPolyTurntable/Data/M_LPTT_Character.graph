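# Character display material for the low-poly turntable tech-art tool, showing flat color, wireframe, UV seam or single-texture views.
# Scalar parameters blend between views: a wireframe texture darkens lines over the color, thresholded Seams channels paint red, and a scaled, brightness-adjusted texture can replace both.
# Lerps route the result between BaseColor and Emissive and adjust metallic, specular and roughness, with manual overrides for the latter two.
# A normal-map parameter feeds Normal.
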Material /Game/NeverCook/ArtTools/Tech_Art/LowPolyTurntable/Data/M_LPTT_Character
graph {
 properties {bAutomaticallySetUsageInEditor=false, bUseMaterialAttributes=true, usedShadingModels=Default Lit}
 nodes [
  n1 = VectorParameter() {parameterName=Color, defaultValue=<rgba 0.333 0.333 0.333 0>}
  n2 = ComponentMask(None=NamedRerouteUsage()) {r=true}
  n3 = LinearInterpolate(A=@n1.RGB, B=Constant3Vector(), Alpha=Saturate(None=Multiply(A=@n2) {constB=4}))
  n4 = TextureSampleParameter2D() {automaticViewMipBias=false, parameterName=Seams}
  n5 = Add(A=@n4.G, B=@n4.G)
  n6 = Power(Base=Subtract(A=@n4.R, B=Saturate(None=@n5))) {constExponent=15}
  n7 = LinearInterpolate(A=@n3, B=Constant3Vector() {constant=<rgba 1 0 0 0>}, Alpha=Ceil(None=Subtract(A=@n6) {constB=0.1}))
  n8 = ScalarParameter() {parameterName=ShowSeams}
  n9 = Multiply(A=TextureCoordinate(), B=ScalarParameter() {defaultValue=1, parameterName=Scale})
  n10 = ScalarParameter() {parameterName=ShowSingleTexture}
  n11 = LinearInterpolate(
   A=LinearInterpolate(A=@n1.RGB, B=LinearInterpolate(A=@n3, B=@n7, Alpha=@n8) {desc=LerpToSeams}, Alpha=ScalarParameter() {parameterName=ShowWireframe})
   B=Multiply(
    A=TextureSampleParameter2D(UVs=@n9) {automaticViewMipBias=false, parameterName=SingleTexture}.RGB
    B=ScalarParameter() {defaultValue=1, parameterName=TextureBrightness}
   ) {constB=0.4}
   Alpha=@n10
  )
  n12 = Constant3Vector()
  n13 = NamedRerouteUsage()
  n14 = Constant()
  n15 = ScalarParameter() {parameterName=ShowAsEmissive}
  n16 = LinearInterpolate(A=Constant() {r=0.5}, B=@n14, Alpha=@n15)
  n17 = ScalarParameter() {parameterName=ShowWireframe}
  n18 = ScalarParameter() {parameterName=UseManualValues}
  n19 = LinearInterpolate(A=Constant() {r=0.5}, B=@n14, Alpha=@n15)
  n20 = TextureSampleParameter2D() {automaticViewMipBias=false, parameterName=Normals, samplerType=SAMPLERTYPE_Normal, texture=/Game/UI/DamageText/Default/+-Normal.+-Normal}
  MakeMaterialAttributes!(
   BaseColor=LinearInterpolate(A=@n11, B=@n12, Alpha=@n13)
   Metallic=LinearInterpolate(A=Constant() {r=0.5}, B=@n14, Alpha=@n15)
   Specular=LinearInterpolate(A=LinearInterpolate(A=@n16, Alpha=@n17) {constB=0.25}, B=ScalarParameter() {parameterName=Specular}, Alpha=@n18)
   Roughness=LinearInterpolate(A=LinearInterpolate(A=@n19, Alpha=@n17) {constB=0.8}, B=ScalarParameter() {parameterName=Roughness}, Alpha=@n18)
   Anisotropy
   EmissiveColor=LinearInterpolate(A=@n12, B=@n11, Alpha=@n13)
   Opacity
   OpacityMask
   Normal=@n20.RGB
   Tangent
   WorldPositionOffset
   SubsurfaceColor
   ClearCoat
   ClearCoatRoughness
   AmbientOcclusion
   Refraction
   CustomizedUVs_0
   CustomizedUVs_1
   CustomizedUVs_2
   CustomizedUVs_3
   CustomizedUVs_4
   CustomizedUVs_5
   CustomizedUVs_6
   CustomizedUVs_7
   PixelDepthOffset
   ShadingModel
   Displacement
  )
  n21 = NamedRerouteDeclaration(None=TextureSampleParameter2D() {automaticViewMipBias=false, parameterName=Wireframe}.R) {name=Wireframe R, nodeColor=<rgba 1 0 0.0937 1>}
  NamedRerouteDeclaration(None=Saturate(None=Multiply(A=@n21) {constB=2})) {name=WireframeView, nodeColor=<rgba 0 1 0.4609 1>}
  NamedRerouteDeclaration(None=@n10) {name=ShowSingleTextureReroute, nodeColor=<rgba 1 0 0.6328 1>}
  NamedRerouteDeclaration(None=@n15) {name=ShowAsEmissive, nodeColor=<rgba 0.2656 1 0 1>}
 ]
}
