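# Editor-utility static mesh actor for the Low Poly Turntable art tool, displaying an imported model in switchable review modes.
# Initialize stores import settings, maps slots to merge, then builds an info object and dynamic material per slot.
# Textures match by slot keyword and suffix; unless quick loading, Python generates wireframe, UV-island and seam textures from the FBX.
# Material_* functions switch views through material parameters, toggling an unlit post-process for true-value display.
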EditorUtilityBlueprint /Game/NeverCook/ArtTools/Tech_Art/LowPolyTurntable/Data/LPTT_StaticMeshActor
graphs [
 CompileMaterialTexturesBP {nodes [
   n1 = CompileMaterialTexturesBP!() -> (then, MaterialInfo:"BP LPTT Material Info Object Reference (by ref)") {is=title K2Node_FunctionEntry}
   > Set SearchKeyword(SearchKeyword:String=Get SlotKeyword!(self:BP LPTT Material Info Object Reference=@n1) -> (SlotKeyword:String)) -> (Output_Get:String)
   > Add!(
    execute
    TargetMap:"Map of Strings to Float (double-precision)s"=Get WireframeWidths() -> (WireframeWidths:"Map of Strings to Float (double-precision)s")
    Key:"String (by ref)"=Get SlotKeyword(self=Get MaterialInfo() -> (MaterialInfo:"BP LPTT Material Info Object Reference (by ref)"))
    Value:"Float (double-precision) (by ref)"=Get DilateSize() -> (DilateSize:double)
   ) -> (then) {is=title K2Node_CallFunction}
   > n2 = Branch(Condition=Get IsQuickLoading?() -> (IsQuickLoading?:bool))
   n3 = For Each Loop(
    Exec=[@n19.Completed, @n2.then]
    Array:Array of Texture 2D Object References=Get Textures() -> (Textures:Array of Texture 2D Object References)
   ) -> (Array Element:Texture 2D Object Reference)
   > n4 = Set CurrentTexture(CurrentTexture:Texture 2D Object Reference=@n3) -> (Output_Get:Texture 2D Object Reference)
   n5 = GetObjectName(Object=@n4)
   n6 = Get SearchKeyword() -> (SearchKeyword:String)
   > Branch(Condition=Contains(SearchIn=ToLower!(SourceString:String=@n5) -> (ReturnValue:String), Substring=ToLower(SourceString=@n6), bSearchFromEnd=true))
   > n7 = Sequence!(execute) -> (then_0, then_1, then_2:Exec, then_3:Exec)
   n8 = Get CurrentTexture() -> (CurrentTexture:Texture 2D Object Reference)
   Branch(execute=@n7.then_2, Condition=EndsWith(SourceString=GetDisplayName(Object=@n8), InSuffix=_N))
   > Set Normal!(execute, Normal:Texture 2D Object Reference=@n8, self:BP LPTT Material Info Object Reference=Get MaterialInfo()) -> (then, Output_Get:Texture 2D Object Reference)
   n9 = Branch(execute=@n7.then_3, Condition=Get IsQuickLoading?())
   n10 = Get CurrentTexture()
   Branch(execute=@n9.else, Condition=EndsWith(SourceString=GetDisplayName(Object=@n10), InSuffix=_VC))
   > Set SRGB(SRGB=true, self=@n10)
   > Set VertexColor!(execute, VertexColor:Texture 2D Object Reference=@n10, self:BP LPTT Material Info Object Reference=Get MaterialInfo()) -> (then, Output_Get:Texture 2D Object Reference)
   n11 = Get CurrentTexture()
   Branch(execute=@n7.then_0, Condition=EndsWith(SourceString=GetDisplayName(Object=@n11), InSuffix=_AO))
   > Set SRGB(SRGB=true, self=@n11)
   > Set AmbientOcclusion!(execute, AmbientOcclusion:Texture 2D Object Reference=@n11, self:BP LPTT Material Info Object Reference=Get MaterialInfo()) -> (then, Output_Get:Texture 2D Object Reference)
   n12 = Branch(execute=@n7.then_1, Condition=Get IsQuickLoading?())
   n13 = Get CurrentTexture()
   Branch(execute=@n12.else, Condition=EndsWith(SourceString=GetDisplayName(Object=@n13), InSuffix=_CRV))
   > Set Curvature!(execute, Curvature:Texture 2D Object Reference=@n13, self:BP LPTT Material Info Object Reference=Get MaterialInfo()) -> (then, Output_Get:Texture 2D Object Reference)
   n14 = Execute Python Script!(
    execute=@n2.else
    PythonScript="import unreal\r\nimport unpipe\r\nimport importlib\r\n\r\nunpipe.requires_plugin(\"epic-shrd-techart\")\r\nfrom epic_shrd_techart.intg.common.Render import wireframe\r\nimportlib.reload(wireframe)\r\nfrom epic_shrd_techart.intg.unreal.helpers import lptt_utils\r\nimportlib.reload(lptt_utils)\r\n\r\ntextures = lptt_utils.generate_lptt_textures_texture2D(\r\n    fbx_path,\r\n    material_wireframe_widths=material_wireframe_widths,\r\n    materials_to_merge=materials_to_merge,\r\n    force_create_wire = force_wireframe,\r\n    force_create_colorized = force_colorized,\r\n    force_create_seams = force_seams,\r\n    ue_asset_path=ue_asset_path\r\n)\r\n\r\nwireframe_list = textures.get('wireframe')\r\ncolorized_list = textures.get('colorized')\r\nseams_list = textures.get('seams')"
    fbx_path:String=Break File Path!(FilePath:"File Path Structure (by ref)"=Get FBX_FilePath() -> (FBX_FilePath:File Path Structure)) -> (FilePath:String) {is=title K2Node_BreakStruct}
    material_wireframe_widths:"Map of Strings to Float (double-precision)s"=Get WireframeWidths()
    materials_to_merge:Map of Strings to Strings=Get SlotsToMerge() -> (SlotsToMerge:Map of Strings to Strings)
    force_wireframe:bool=Get ForceCreateWireframe() -> (ForceCreateWireframe:bool)
    force_colorized:bool=Get ForceCreateColorized_UVs() -> (ForceCreateColorized_UVs:bool)
    force_seams:bool=Get ForceCreateSeams() -> (ForceCreateSeams:bool)
    ue_asset_path:String=Append(A=/Game/Temp/, B=Get AssetName() -> (AssetName:String))
   ) -> (
    then
    ReturnValue
    wireframe_list:Array of Texture 2D Object References
    colorized_list:Array of Texture 2D Object References
    seams_list:Array of Texture 2D Object References
   ) {
    inputs [fbx_path, material_wireframe_widths, materials_to_merge, force_wireframe, force_colorized, force_seams, ue_asset_path]
    outputs [wireframe_list, colorized_list, seams_list]
   }
   > n15 = For Each Loop with Break!(Exec:Exec, Array:Array of Texture 2D Object References=@n14.wireframe_list, Break:Exec=@n16.then) -> (LoopBody:Exec, Array Element:Texture 2D Object Reference, Array Index:int, Completed:Exec)
   > Branch(Condition=Contains(SearchIn=GetDisplayName(Object=@n15), Substring=Get SearchKeyword()))
   > Set Topology!(execute, Topology:Texture 2D Object Reference=@n15, self:BP LPTT Material Info Object Reference=Get MaterialInfo()) -> (then, Output_Get:Texture 2D Object Reference)
   > n16 = Set ForceCreateWireframe(ForceCreateWireframe:bool=false) -> (Output_Get:bool)
   n17 = For Each Loop with Break(Exec=@n15.Completed, Array=@n14.colorized_list, Break=@n18.then)
   > Branch(Condition=Contains(SearchIn=GetDisplayName(Object=@n17), Substring=Get SearchKeyword()))
   > Set ColorizedUVs!(execute, ColorizedUVs:Texture 2D Object Reference=@n17, self:BP LPTT Material Info Object Reference=Get MaterialInfo()) -> (then, Output_Get:Texture 2D Object Reference)
   > n18 = Set ForceCreateColorized_UVs(ForceCreateColorized_UVs:bool=false) -> (Output_Get:bool)
   n19 = For Each Loop with Break(Exec=@n17.Completed, Array=@n14.seams_list, Break=@n20.then)
   > Branch(Condition=Contains(SearchIn=GetDisplayName(Object=@n19), Substring=Get SearchKeyword()))
   > Set Seams!(execute, Seams:Texture 2D Object Reference=@n19, self:BP LPTT Material Info Object Reference=Get MaterialInfo()) -> (then, Output_Get:Texture 2D Object Reference)
   > n20 = Set ForceCreateSeams(ForceCreateSeams:bool=false) -> (Output_Get:bool)
  ]}
 DetermineSlotsToMerge {nodes [
   DetermineSlotsToMerge() {is=title K2Node_FunctionEntry}
   > n1 = Sequence!(execute) -> (then_0, then_1)
   > n2 = For Each Loop!(
    Exec
    Array:Array of LPTT Import Material Slot Struct Structures=Get MaterialSlots() -> (MaterialSlots:Array of LPTT Import Material Slot Struct Structures)
   ) -> (
    LoopBody
    Array Element_SlotName_23_#1:String
    Array Element_MaterialType_26_#2:LPTT_MaterialSlotType Enum
    Array Element_CustomMaterialType_22_#3:String
    Array Index
    Completed
   )
   > n3 = Sequence!(execute) -> (then_0, then_1, then_2:Exec)
   Branch(
    execute=@n3.then_2
    Condition="Not Equal, Case Insenstive (String)"!(A:String=@n2, B:String=Get SlotKeyword!() -> (SlotKeyword)) -> (ReturnValue:bool)
   )
   > Add(TargetMap:Map of Strings to Strings=Get SlotsToMerge(), Key=@n2, Value:"String (by ref)"=Get SlotKeyword())
   Branch(execute=@n3.then_0, Condition="Equal, Case Insensitive (String)"(A=@n2, B=Eyes))
   > Set ContainsEyes(ContainsEyes:bool=true) -> (Output_Get:bool)
   n4 = Branch(
    execute=@n3.then_1
    Condition="Equal (Enum)"(
     A:LPTT_MaterialSlotType Enum=@n2.Array Element_MaterialType_26_#2
     B:LPTT_MaterialSlotType Enum=NewEnumerator10
    )
   )
   > Set SlotKeyword(SlotKeyword:String=@n2.Array Element_CustomMaterialType_22_#3) -> (Output_Get:String)
   n5 = Branch(execute=@n4.else, Condition="Equal (Enum)"(A=@n2.Array Element_MaterialType_26_#2, B=NewEnumerator0))
   > Set SlotKeyword(SlotKeyword=@n2)
   Set SlotKeyword(
    execute=@n5.else
    SlotKeyword=Enum to String!(Enumerator:LPTT_MaterialSlotType Enum=@n2.Array Element_MaterialType_26_#2) -> (ReturnValue:String)
   )
   n6 = Switch on LPTT_ImportPart_PathGen!(execute=@n1.then_1, Selection:LPTT_ImportPart_PathGen Enum=Get PartType() -> (PartType:LPTT_ImportPart_PathGen Enum=NewEnumerator0)) -> (
    NewEnumerator0:Exec
    NewEnumerator1:Exec
    NewEnumerator2:Exec
    NewEnumerator4:Exec
    NewEnumerator7:Exec
    NewEnumerator12:Exec
    NewEnumerator13:Exec
    NewEnumerator3:Exec
    NewEnumerator8:Exec
    NewEnumerator10:Exec
   ) {is=title K2Node_SwitchEnum}
   Branch(execute=@n6.NewEnumerator2, Condition=Get ContainsEyes() -> (ContainsEyes:bool))
   > Add(TargetMap=Get SlotsToMerge(), Key=Eyes, Value=Head)
   Branch(execute=@n6.NewEnumerator3, Condition=Get ContainsEyes())
   > Add(TargetMap=Get SlotsToMerge(), Key=Eyes, Value=Body)
  ]}
 EventGraph {nodes [
   Event BeginPlay()
   Event ActorBeginOverlap()
   Event Tick!() -> (OutputDelegate, then, DeltaSeconds:float)
  ]}
 GetFlowProjectInfo {nodes [
   GetFlowProjectInfo() {is=title K2Node_FunctionEntry}
   > n1 = Execute Python Script!(
    execute
    PythonScript="import unpipe\r\nfrom unpipe.util import get_active_pipeline\r\n\r\ncontext = get_active_pipeline().get_context()\r\n\r\ntry:\r\n    id = int(context[\"project_id\"])\r\nexcept:\r\n    id = None\r\n\r\ntry:\r\n    name = context[\"name\"]\r\nexcept:\r\n    name = None"
   ) -> (then, ReturnValue, id:int, name:Name) {outputs [id,name]}
   > Return Node!(execute, ID:int=@n1.id, Name:Name=@n1.name)
  ]}
 Initialize {nodes [
   n1 = Initialize!() -> (
    then
    Mesh:Static Mesh Object Reference
    FBX_FilePath:File Path Structure
    Textures:Array of Texture 2D Object References
    MaterialSlots:Array of LPTT Import Material Slot Struct Structures
    NormalsImportSettings:Enum_LPTT_NormalsImportSettings Enum=NewEnumerator0
    DispName:String
    AssetName:String
    DilateSize:int
    PartType:LPTT_ImportPart_PathGen Enum=NewEnumerator0
    PivotStyle:LPTT_PivotLocation_Enum Enum=NewEnumerator0
    ForceWireframe:bool=~
    ForceColorizedUVs:bool=~
    ForceSeams:bool=~
    QuickLoad:bool=~
   ) {is=title K2Node_FunctionEntry}
   > SetStaticMesh(self=Get StaticMeshComponent!() -> (StaticMeshComponent), NewMesh=@n1)
   > Set MeshRef(MeshRef:Static Mesh Object Reference=@n1) -> (Output_Get:Static Mesh Object Reference)
   > Set FBX_FilePath(FBX_FilePath:File Path Structure=@n1.FBX_FilePath) -> (Output_Get:File Path Structure)
   > Set Textures(Textures:Array of Texture 2D Object References=@n1.Textures) -> (Output_Get:Array of Texture 2D Object References)
   > Set MaterialSlots(MaterialSlots:Array of LPTT Import Material Slot Struct Structures=@n1.MaterialSlots) -> (Output_Get:Array of LPTT Import Material Slot Struct Structures)
   > Set NormalsImportSetting(NormalsImportSetting:Enum_LPTT_NormalsImportSettings Enum=@n1.NormalsImportSettings) -> (Output_Get:Enum_LPTT_NormalsImportSettings Enum=NewEnumerator0)
   > Set DisplayName(DisplayName:String=@n1.DispName) -> (Output_Get:String)
   > Set AssetName(AssetName:String=@n1.AssetName) -> (Output_Get:String)
   > Set DilateSize(DilateSize:double="To Float (Integer)"(InInt=@n1.DilateSize)) -> (Output_Get:double)
   > Set PartType(PartType:LPTT_ImportPart_PathGen Enum=@n1.PartType) -> (Output_Get:LPTT_ImportPart_PathGen Enum=NewEnumerator0)
   > Set PivotStyle(PivotStyle:LPTT_PivotLocation_Enum Enum=@n1.PivotStyle) -> (Output_Get:LPTT_PivotLocation_Enum Enum=NewEnumerator0)
   > Set ForceCreateWireframe(ForceCreateWireframe=@n1.ForceWireframe)
   > Set ForceCreateColorized_UVs(ForceCreateColorized_UVs=@n1.ForceColorizedUVs)
   > Set ForceCreateSeams(ForceCreateSeams=@n1.ForceSeams)
   > Set IsQuickLoading?(IsQuickLoading?:bool=@n1.QuickLoad) -> (Output_Get:bool)
   > InitializeMaterials()
   > UpdateVertexCount()
  ]}
 InitializeMaterials {nodes [
   InitializeMaterials() {is=title K2Node_FunctionEntry}
   > DetermineSlotsToMerge() {is=title K2Node_CallFunction}
   > n1 = Sequence!(execute) -> (then_0, then_1)
   > n2 = For Each Loop(Array=Get MaterialSlots())
   > n3 = Sequence!(execute) -> (then_0, then_1, then_2:Exec, then_3:Exec)
   > n4 = Branch(Condition="Equal (Enum)"(A=@n2.Array Element_MaterialType_26_#2, B=NewEnumerator10))
   > n5 = Set SlotKeyword(SlotKeyword=@n2.Array Element_CustomMaterialType_22_#3)
   n6 = Construct Object from Class(
    execute=[@n9.then, @n5.then, @n8.then]
    Class=&/BP_LPTT_MaterialInfo.BP_LPTT_MaterialInfo_C
   ) -> (ReturnValue:BP LPTT Material Info Object Reference)
   > Set SlotName!(execute, SlotName:String=@n2, self:BP LPTT Material Info Object Reference=@n6) -> (then, Output_Get:String)
   > Set SlotKeyword!(execute, SlotKeyword=Get SlotKeyword(), self:BP LPTT Material Info Object Reference=@n6) -> (then, Output_Get)
   > Set PartType!(
    execute
    PartType:LPTT_MaterialSlotType Enum=@n2.Array Element_MaterialType_26_#2
    self:BP LPTT Material Info Object Reference=@n6
   ) -> (then, Output_Get:LPTT_MaterialSlotType Enum=NewEnumerator0)
   > Set MatInfoBP(MatInfoBP:BP LPTT Material Info Object Reference=@n6) -> (Output_Get:BP LPTT Material Info Object Reference)
   n7 = Branch(execute=@n4.else, Condition="Equal (Enum)"(A=@n2.Array Element_MaterialType_26_#2, B=NewEnumerator0))
   > n8 = Set SlotKeyword!(execute, SlotKeyword=@n2) -> (then, Output_Get)
   n9 = Set SlotKeyword(execute=@n7.else, SlotKeyword=Enum to String(Enumerator=@n2.Array Element_MaterialType_26_#2))
   CompileMaterialTexturesBP!(
    execute=@n3.then_1
    self
    MaterialInfo:"BP LPTT Material Info Object Reference (by ref)"=Get MatInfoBP() -> (MatInfoBP:BP LPTT Material Info Object Reference)
   ) -> (then) {is=title K2Node_CallFunction}
   n10 = GetFlowProjectInfo!(execute=@n3.then_2, self) -> (then, ID:int=0, Name:Name=None) {is=title K2Node_CallFunction}
   n11 = Get PartType() -> (PartType=NewEnumerator0)
   > n12 = Branch(Condition=AND Boolean!(
     A="Equal (Enum)"(A:LPTT_ImportPart_PathGen Enum=@n11, B:LPTT_ImportPart_PathGen Enum=NewEnumerator2)
     B="Equal, Case Insensitive (String)"(A=@n2, B=eyes)
     C:bool="Equal (Name)"(A=@n10.Name, B=Fortnite)
    ) -> (ReturnValue))
   n13 = Get MatInfoBP()
   > Set PartType(PartType=NewEnumerator0, self=@n13) -> (Output_Get=NewEnumerator0)
   > n14 = Set MaterialSource!(
    execute
    MaterialSource:Material Interface Object Reference=&/MI_MAT_Turntable_ModelChecker_Eyes
    self:BP LPTT Material Info Object Reference=@n13
   ) -> (then, Output_Get:Material Interface Object Reference)
   n15 = GetDisplayName(Object=Get StaticMesh(self=Get StaticMeshComponent()))
   n16 = Get SlotName!(self:BP LPTT Material Info Object Reference=@n13) -> (SlotName:String)
   n17 = Get MaterialSource!(self:BP LPTT Material Info Object Reference=Get MatInfoBP()) -> (MaterialSource:Material Interface Object Reference)
   n18 = CreateDynamicMaterialInstance(
    execute=[@n14.then, @n22.then]
    Parent=@n17
    OptionalName=String To Name(InString=Append!(A=@n15, B=_, C:String=@n16) -> (ReturnValue))
    CreationFlags=None
   )
   > n19 = Set MaterialInstance(MaterialInstance:Material Instance Dynamic Object Reference=@n18) -> (Output_Get:Material Instance Dynamic Object Reference)
   > n20 = Branch(Condition="Equal (Object)"(A=@n17, B=&/M_LPTT_Character))
   > n21 = SetTextureParameterValue(
    self=@n19
    ParameterName=Normals
    Value=Get Normal!(self:BP LPTT Material Info Object Reference=Get MatInfoBP()) -> (Normal:Texture 2D Object Reference)
   )
   Set Material!(
    execute=[@n21.then, @n20.else]
    Material:Material Instance Dynamic Object Reference=@n18
    self:BP LPTT Material Info Object Reference=Get MatInfoBP()
   ) -> (then, Output_Get:Material Instance Dynamic Object Reference)
   n22 = Set MaterialSource(execute=@n12.else, MaterialSource=&M_LPTT_Character, self=@n13)
   Add(
    execute=@n3.then_3
    TargetArray:Array of BP LPTT Material Info Object References=Get MaterialsBP() -> (MaterialsBP:Array of BP LPTT Material Info Object References)
    NewItem:"BP LPTT Material Info Object Reference (by ref)"=Get MatInfoBP()
   ) {is=title K2Node_CallArrayFunction}
   > Add(
    TargetArray:Array of Material Instance Dynamic Object References=Get MaterialList() -> (MaterialList:Array of Material Instance Dynamic Object References)
    NewItem:"Material Instance Dynamic Object Reference (by ref)"=Get MaterialInstance() -> (MaterialInstance:Material Instance Dynamic Object Reference)
   )
   SetComponentMaterialList!(
    execute=@n1.then_1
    Component:Primitive Component Object Reference=Get StaticMeshComponent()
    MaterialList:Array of Material Interface Object References=Get MaterialList()
   ) -> (then)
   > Material_Greybox()
  ]}
 Material_AO {nodes [
   Material_AO() {is=title K2Node_FunctionEntry}
   > n1 = For Each Loop!(Exec, Array:Array of BP LPTT Material Info Object References=Get MaterialsBP()) -> (LoopBody, Array Element:BP LPTT Material Info Object Reference, Array Index, Completed)
   n2 = Get Material!(self:BP LPTT Material Info Object Reference=@n1) -> (Material:Material Instance Dynamic Object Reference)
   > SetTextureParameterValue(self=@n2, ParameterName=Normals, Value=/Engine/EngineMaterials/BaseFlattenNormalMap)
   > ShowPregeneratedTexture!(
    execute
    self
    Material:Material Instance Dynamic Object Reference=@n2
    Value:Texture Object Reference=Get AmbientOcclusion!(self:BP LPTT Material Info Object Reference=@n1) -> (AmbientOcclusion:Texture 2D Object Reference)
    Scale:float=1
    Value_Secondary:Texture Object Reference
    UsePrimaryForSecondary:bool=true
    DisplayAsEmissive:bool=false
    TextureBrightness:double=1
    DisplayAsTrueValues:bool=true
    UseManualSpecRough:bool=false
    Specular:double=0
    Roughness:double=0
   ) -> (then)
  ]}
 Material_CRV {nodes [
   Material_CRV() {is=title K2Node_FunctionEntry}
   > n1 = For Each Loop(Array=Get MaterialsBP())
   > ShowPregeneratedTexture(
    Material=Get Material(self=@n1)
    Value=Get Curvature!(self:BP LPTT Material Info Object Reference=@n1) -> (Curvature:Texture 2D Object Reference)
    Scale=1
    UsePrimaryForSecondary=true
    DisplayAsEmissive=true
    TextureBrightness=1
    DisplayAsTrueValues=false
    UseManualSpecRough=false
    Specular=0
    Roughness=0
   )
  ]}
 Material_Checker {nodes [
   Material_Checker() {is=title K2Node_FunctionEntry}
   > n1 = For Each Loop(Array=Get MaterialsBP())
   n2 = Get Material(self=@n1)
   > SetTextureParameterValue(self=@n2, ParameterName=Normals, Value=Get Normal(self=@n1))
   > ShowPregeneratedTexture(
    Material=@n2
    Value=&/UV_checker_Map_byValle
    Scale=11
    UsePrimaryForSecondary=true
    DisplayAsEmissive=false
    TextureBrightness=0.55
    DisplayAsTrueValues=false
    UseManualSpecRough=true
    Specular=0.5
    Roughness=0.7
   )
  ]}
 Material_Greybox {nodes [
   Material_Greybox() {is=title K2Node_FunctionEntry}
   > Set bEnabled!(
    execute
    bEnabled:bool=false
    self:Post Process Component Object Reference=Get PregeneratedUnlitPP() -> (PregeneratedUnlitPP:Post Process Component Object Reference)
   ) -> (then, Output_Get:bool)
   > n1 = For Each Loop(Array=Get MaterialsBP())
   n2 = Get Material(self=@n1)
   > SetTextureParameterValue(self=@n2, ParameterName=Normals, Value=Get Normal(self=@n1))
   > SetScalarParameterValue(self=@n2, ParameterName=ShowWireframe, Value=0)
   > SetScalarParameterValue(self=@n2, ParameterName=ShowSingleTexture, Value=0)
   > SetScalarParameterValue(self=@n2, ParameterName=ShowAsEmissive, Value=0)
   > SetScalarParameterValue(self=@n2, ParameterName=Roughness, Value=0.55)
   > SetScalarParameterValue(self=@n2, ParameterName=UseManualValues, Value=0)
  ]}
 Material_Islands {nodes [
   Material_Islands() {is=title K2Node_FunctionEntry}
   > n1 = For Each Loop(Array=Get MaterialsBP())
   > ShowPregeneratedTexture(
    Material=Get Material(self=@n1)
    Value=Get ColorizedUVs!(self:BP LPTT Material Info Object Reference=@n1) -> (ColorizedUVs:Texture 2D Object Reference)
    Scale=1
    UsePrimaryForSecondary=true
    DisplayAsEmissive=false
    TextureBrightness=1
    DisplayAsTrueValues=false
    UseManualSpecRough=false
    Specular=0
    Roughness=0
   )
  ]}
 Material_VertexColor {nodes [
   Material_VertexColor() {is=title K2Node_FunctionEntry}
   > n1 = For Each Loop(Array=Get MaterialsBP())
   n2 = Get Material(self=@n1)
   > SetTextureParameterValue(self=@n2, ParameterName=Normals, Value=&BaseFlattenNormalMap)
   > ShowPregeneratedTexture(
    Material=@n2
    Value=Get VertexColor!(self:BP LPTT Material Info Object Reference=@n1) -> (VertexColor:Texture 2D Object Reference)
    Scale=1
    UsePrimaryForSecondary=false
    DisplayAsEmissive=false
    TextureBrightness=1
    DisplayAsTrueValues=true
    UseManualSpecRough=false
    Specular=0
    Roughness=0
   )
  ]}
 Material_Wireframe {nodes [
   SetMaterialInstanceScalarParameterValue!(
    execute
    Instance:Material Instance Constant Object Reference
    ParameterName:Name=ShowSeams
    Value:float=0
    Association:EMaterialParameterAssociation Enum=GlobalParameter
   ) -> (then, ReturnValue:bool)
   > SetMaterialInstanceTextureParameterValue!(
    execute
    Instance:Material Instance Constant Object Reference
    ParameterName:Name=Wireframe
    Value:Texture Object Reference
    Association:EMaterialParameterAssociation Enum=GlobalParameter
   ) -> (then, ReturnValue:bool)
   > SetMaterialInstanceTextureParameterValue(ParameterName=Overlaps, Association=GlobalParameter)
   > SetMaterialInstanceTextureParameterValue(ParameterName=Seams, Association=GlobalParameter)
   > SetMaterialInstanceScalarParameterValue(ParameterName=ShowWireframe, Value=1, Association=GlobalParameter)
   > SetMaterialInstanceScalarParameterValue(ParameterName=ShowSingleTexture, Value=0, Association=GlobalParameter)
   > SetMaterialInstanceScalarParameterValue(ParameterName=ShowAsEmissive, Value=0, Association=GlobalParameter)
   > SetMaterialInstanceScalarParameterValue(ParameterName=UseManualValues, Value=0, Association=GlobalParameter)
   > SetMaterialInstanceScalarParameterValue(ParameterName=None, Value=0, Association=GlobalParameter)
   Material_Wireframe!() -> (then, ShowSeams?:bool=~) {is=title K2Node_FunctionEntry}
   > Set bEnabled(bEnabled=false, self=Get PregeneratedUnlitPP())
   > n1 = For Each Loop(Array=Get MaterialsBP())
   > Set CurrentTopologyTexture(CurrentTopologyTexture:Texture 2D Object Reference=Get Topology!(self:BP LPTT Material Info Object Reference=@n1) -> (Topology:Texture 2D Object Reference)) -> (Output_Get:Texture 2D Object Reference)
   > Set CurrentSeamTexture(CurrentSeamTexture:Texture 2D Object Reference=Get Seams!(self:BP LPTT Material Info Object Reference=@n1) -> (Seams:Texture 2D Object Reference)) -> (Output_Get:Texture 2D Object Reference)
   n2 = Get Material(self=@n1)
   > SetTextureParameterValue(self=@n2, ParameterName=Normals, Value=Get Normal(self=@n1))
   n3 = Get CurrentTopologyTexture() -> (CurrentTopologyTexture:Texture 2D Object Reference)
   > Is Valid(InputObject=@n3)
   > SetScalarParameterValue(self=@n2, ParameterName=ShowSeams, Value="To Float (Boolean)"(InBool=Get ShowSeams?() -> (ShowSeams?:bool)))
   > SetTextureParameterValue(self=@n2, ParameterName=Wireframe, Value=@n3)
   > SetTextureParameterValue(self=@n2, ParameterName=Seams, Value=Get CurrentSeamTexture() -> (CurrentSeamTexture:Texture 2D Object Reference))
   > SetScalarParameterValue(self=@n2, ParameterName=ShowWireframe, Value=1)
   > SetScalarParameterValue(self=@n2, ParameterName=ShowSingleTexture, Value=0)
   > SetScalarParameterValue(self=@n2, ParameterName=ShowAsEmissive, Value=0)
   > SetScalarParameterValue(self=@n2, ParameterName=UseManualValues, Value=0)
  ]}
 ShowPregeneratedTexture {nodes [
   n1 = ShowPregeneratedTexture!() -> (
    then
    Material:Material Instance Dynamic Object Reference
    Value:Texture Object Reference
    Scale:float=1
    Value_Secondary:Texture Object Reference
    UsePrimaryForSecondary:bool=~
    DisplayAsEmissive:bool=~
    TextureBrightness:double=1
    DisplayAsTrueValues:bool=~
    UseManualSpecRough:bool=~
    Specular:double=~
    Roughness:double=~
   ) {is=title K2Node_FunctionEntry}
   > Set bEnabled(bEnabled=@n1.DisplayAsTrueValues, self=Get PregeneratedUnlitPP())
   > Is Valid(InputObject=@n1.Value)
   > n2 = Set Texture(Texture:Texture Object Reference=@n1.Value) -> (Output_Get:Texture Object Reference)
   > n3 = Branch(Condition=OR Boolean(
     A=@n1.UsePrimaryForSecondary
     B="Equal (Object)"(A=@n1.Value_Secondary, B=/Game/Global/Textures/Default/Blanks/T_Blank_Mask_Magenta_R1G0B1A1)
    ))
   > n4 = Set SecondarySlotTexture(SecondarySlotTexture:Texture Object Reference=@n2) -> (Output_Get:Texture Object Reference)
   n5 = Get Material!() -> (Material)
   SetScalarParameterValue(execute=[@n4.then, @n6.then], self=@n5, ParameterName=Scale, Value=@n1.Scale)
   > SetScalarParameterValue(self=@n5, ParameterName=ShowAsEmissive, Value="To Float (Boolean)"(InBool=@n1.DisplayAsEmissive))
   > SetScalarParameterValue(self=@n5, ParameterName=ShowWireframe, Value=0)
   > SetScalarParameterValue(self=@n5, ParameterName=ShowSingleTexture, Value=1)
   > SetTextureParameterValue(self=@n5, ParameterName=SingleTexture, Value=Get Texture() -> (Texture:Texture Object Reference))
   > SetScalarParameterValue(self=@n5, ParameterName=TextureBrightness, Value=@n1.TextureBrightness)
   > SetScalarParameterValue(self=@n5, ParameterName=UseManualValues, Value=1)
   > SetScalarParameterValue(self=@n5, ParameterName=Specular, Value=0)
   > SetScalarParameterValue(self=@n5, ParameterName=Roughness, Value=@n1.Roughness)
   > Return Node!(execute)
   n6 = Set SecondarySlotTexture(execute=@n3.else, SecondarySlotTexture=@n1.Value_Secondary)
   Reroute Node(InputPin:bool=@n1.UseManualSpecRough) -> (OutputPin:bool)
   Reroute Node(InputPin:double=@n1.Specular) -> (OutputPin:double)
  ]}
 UpdateVertexCount {nodes [
   UpdateVertexCount() {is=title K2Node_FunctionEntry}
   n1 = Get StaticMeshComponent()
   > Set VertexCount(VertexCount:int=GetNumVertices!(self:Static Mesh Object Reference=Get StaticMesh(self=@n1), LODIndex:int=0) -> (ReturnValue:int=0)) -> (Output_Get:int=0)
  ]}
 UserConstructionScript {nodes [Construction Script()]}
]
variables [
 AssetName {type=string, default=""}
 DilateSize {type=number}
 DisplayName {type=string, default=""}
 FBX_FilePath {type=FilePath, default {filePath=""}}
 ForceCreateColorized_UVs {type=boolean}
 ForceCreateSeams {type=boolean}
 ForceCreateWireframe {type=boolean}
 IsQuickLoading? {type=boolean}
 MaterialsBP {type="array<BP_LPTT_MaterialInfo_C>", default []}
 MaterialSlots {type="array<LPTT_Import_MaterialSlot_Struct>", default []}
 MeshRef {type=StaticMesh, default=null}
 NormalsImportSetting {type=Enum_LPTT_NormalsImportSettings, default=ComputeNormalsAndTangents}
 PartType {type=LPTT_ImportPart_PathGen, default=Auto Detect}
 PivotStyle {type=LPTT_PivotLocation_Enum, default=Imported Pivot}
 SlotsToMerge {type="map<string>", default {}}
 Textures {type="array<Texture2D>", default []}
 VertexCount {type=integer}
 WireframeWidths {type="map<number>", default {Body=2, FaceAcc=4, Head=4}}
]
