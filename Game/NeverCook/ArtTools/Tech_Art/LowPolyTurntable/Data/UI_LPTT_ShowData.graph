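# Art-tool overlay widget for the Low Poly Turntable showing user name, date, engine version, pass label and texture previews.
# On Construct it lists vertex counts of Active-tagged mesh actors, marking over-budget parts with rich-text error tags.
# It then maps material ColorizedUVs into Body, Head, FaceAcc and Hair wireframe slots, tagging unrecognized parts into free slots as Custom.
# Separate functions switch previews between wireframe and UV-island textures, set pass text, or hide previews.
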WidgetBlueprint /Game/NeverCook/ArtTools/Tech_Art/LowPolyTurntable/Data/UI_LPTT_ShowData
graphs [
 AssignAdditionalTags {nodes [
   n1 = AssignAdditionalTags!() -> (then, ActiveActors:Array of LPTT Static Mesh Actor Object References, UsedNames:Array of Names) {is=title K2Node_FunctionEntry}
   > n2 = Branch(Condition="integer > integer"(A=Length(TargetArray:Array of Names=@n1.UsedNames), B=4))
   > Return Node()
   n3 = For Each Loop(Exec=@n2.else, Array:Array of Names=@n1.UsedNames) -> (Array Element:Name)
   n4 = For Each Loop(Exec=@n3.Completed, Array=Get UknownActorTags() -> (UknownActorTags:Array of Names))
   > n5 = Branch(Condition=Get IsBodyUsed?() -> (IsBodyUsed?:bool))
   > n6 = Branch(Condition=Get IsHeadUsed?() -> (IsHeadUsed?:bool))
   > n7 = Branch(Condition=Get IsFaceAccUsed?() -> (IsFaceAccUsed?:bool))
   > n8 = Branch(Condition=Get IsHairUsed?() -> (IsHairUsed?:bool))
   > PrintString!(execute, InString=Error - Out of open texture slots to display, bPrintToScreen, bPrintToLog, TextColor, Duration, Key) -> (then)
   Set IsHairUsed?(execute=@n8.else, IsHairUsed?:bool=true) -> (Output_Get:bool)
   > n9 = For Each Loop(Array:Array of LPTT Static Mesh Actor Object References=@n1) -> (Array Element:LPTT Static Mesh Actor Object Reference)
   > Branch(Condition=ActorHasTag!(self:Actor Object Reference=@n9, Tag:Name=@n4) -> (ReturnValue:bool))
   > AddActorTag!(execute, Actor:Actor Object Reference=@n9, Tag:"Name (by ref)"=MakeLiteralName(Value=CustomHair)) -> (then)
   Set IsFaceAccUsed?(execute=@n7.else, IsFaceAccUsed?:bool=true) -> (Output_Get:bool)
   > n10 = For Each Loop(Array=@n1)
   > Branch(Condition=ActorHasTag(self=@n10, Tag=@n4))
   > AddActorTag(Actor=@n10, Tag=MakeLiteralName(Value=CustomFaceAcc))
   Set IsHeadUsed?(execute=@n6.else, IsHeadUsed?:bool=true) -> (Output_Get:bool)
   > n11 = For Each Loop(Array:Array of Actor Object References=@n1) -> (Array Element:Actor Object Reference)
   > Branch(Condition=ActorHasTag(self=@n11, Tag=@n4))
   > AddActorTag(Actor=@n11, Tag=MakeLiteralName(Value=CustomHead))
   Set IsBodyUsed?(execute=@n5.else, IsBodyUsed?:bool=true) -> (Output_Get:bool)
   > n12 = For Each Loop(Array:Array of LPTT Static Mesh Actor Object References=@n1) -> (Array Element:LPTT Static Mesh Actor Object Reference)
   > Branch(Condition=ActorHasTag(self=@n12, Tag=@n4))
   > AddActorTag(Actor=@n12, Tag=MakeLiteralName(Value=CustomBody))
   n13 = Switch on Name!(execute:Exec=@n3.LoopBody, Selection:Name=@n3) -> (Default:Exec, Body:Exec, Backpack:Exec, Shoe:Exec, FaceAcc:Exec, Head:Exec, Hair:Exec, FaceAccExtra:Exec) {is=title K2Node_SwitchName, pinNames [Body,Backpack,Shoe,FaceAcc,Head,Hair,FaceAccExtra]}
   Set IsBodyUsed?(execute=[@n13.Backpack, @n13.Body, @n13.Shoe], IsBodyUsed?=true)
   Add(execute=@n13.Default, TargetArray:Array of Names=Get UknownActorTags(), NewItem:"Name (by ref)"=@n3) {is=title K2Node_CallArrayFunction}
   Set IsFaceAccUsed?(execute=@n13.FaceAcc, IsFaceAccUsed?=true)
   Set IsHairUsed?(execute=[@n13.FaceAccExtra, @n13.Hair], IsHairUsed?=true)
   Set IsHeadUsed?(execute=@n13.Head, IsHeadUsed?=true)
  ]}
 CollectTexturePreviews {nodes [
   CollectTexturePreviews() {is=title K2Node_FunctionEntry}
   > n1 = GetAllActorsOfClassWithTag!(
    execute
    ActorClass:Actor Class Reference=&/LPTT_StaticMeshActor.LPTT_StaticMeshActor_C
    Tag:Name=Active
   ) -> (then, OutActors:Array of LPTT Static Mesh Actor Object References)
   > n2 = CountValidTags!(execute, self, ActiveActors:Array of LPTT Static Mesh Actor Object References=@n1) -> (then, ActorTags1:Array of Names)
   > AssignAdditionalTags!(execute, self, ActiveActors:Array of LPTT Static Mesh Actor Object References=@n1, UsedNames:Array of Names=@n2) -> (then) {is=title K2Node_CallFunction}
   > n3 = For Each Loop(Array=@n1)
   n4 = "For Each Loop (Map)"!(
    execute:Exec=@n3.Completed
    MapPin:Map of Strings to BP LPTT Material Info Object References=Get UniqueTextureSets() -> (UniqueTextureSets:Map of Strings to BP LPTT Material Info Object References)
    BreakPin:Exec
   ) -> (then:Exec, KeyPin:String, ValuePin:BP LPTT Material Info Object Reference, CompletedPin:Exec)
   > n5 = Switch on String!(execute:Exec, Selection:String=@n4) -> (
    Default:Exec
    Body:Exec
    Backpack:Exec
    Shoe:Exec
    CustomBody:Exec
    FaceAcc:Exec
    CustomFaceAcc:Exec
    Accessory:Exec
    Head:Exec
    CustomHead:Exec
    Hair:Exec
    NeckAcc:Exec
    FaceAccExtra:Exec
    CustomHair:Exec
   ) {pinNames [Body, Backpack, Shoe, CustomBody, FaceAcc, CustomFaceAcc, Accessory, Head, CustomHead, Hair, NeckAcc, FaceAccExtra, CustomHair]}
   Sequence(execute=[@n5.Backpack, @n5.Body, @n5.CustomBody, @n5.Shoe])
   n6 = Get ColorizedUVs!(self:BP LPTT Material Info Object Reference=@n4.ValuePin) -> (ColorizedUVs:Texture 2D Object Reference)
   n7 = IsValid(Object=@n6)
   > Branch(Condition=@n7)
   > Set WireFrame_Body(WireFrame_Body:Texture Object Reference=@n6) -> (Output_Get:Texture Object Reference)
   Sequence(execute=[@n5.Accessory, @n5.CustomFaceAcc, @n5.FaceAcc])
   > Branch(Condition=@n7)
   > Set WireFrame_FaceAcc(WireFrame_FaceAcc:Texture Object Reference=@n6) -> (Output_Get:Texture Object Reference)
   Sequence(execute=[@n5.CustomHead, @n5.Head])
   > Branch(Condition=@n7)
   > Set WireFrame_Head(WireFrame_Head:Texture Object Reference=@n6) -> (Output_Get:Texture Object Reference)
   Sequence(execute=[@n5.CustomHair, @n5.FaceAccExtra, @n5.Hair, @n5.NeckAcc])
   > Branch(Condition=@n7)
   > Set WireFrame_Hair(WireFrame_Hair:Texture Object Reference=@n6) -> (Output_Get:Texture Object Reference)
   n8 = For Each Loop(
    Exec=@n3.LoopBody
    Array:Array of BP LPTT Material Info Object References=Get MaterialsBP!(self:LPTT Static Mesh Actor Object Reference=@n3) -> (MaterialsBP:Array of BP LPTT Material Info Object References)
   ) -> (Array Element:BP LPTT Material Info Object Reference)
   n9 = Get SlotKeyword!(self:BP LPTT Material Info Object Reference=@n8) -> (SlotKeyword:String)
   > n10 = Branch(Condition="Equal, Case Insensitive (String)"(A=@n9, B=eyes))
   Is Valid(exec=@n10.else, InputObject=Get ColorizedUVs(self=@n8))
   n11 = Get UniqueTextureSets()
   > n12 = Branch(Condition=Contains!(TargetMap:Map of Strings to BP LPTT Material Info Object References=@n11, Key:"String (by ref)"=@n9) -> (ReturnValue))
   Add!(
    execute=@n12.else
    TargetMap:Map of Strings to BP LPTT Material Info Object References=@n11
    Key:"String (by ref)"=@n9
    Value:"BP LPTT Material Info Object Reference (by ref)"=@n8
   ) -> (then) {is=title K2Node_CallFunction}
  ]}
 CompileAllVertexCounts {nodes [
   CompileAllVertexCounts() {is=title K2Node_FunctionEntry}
   > n1 = GetAllActorsOfClassWithTag(ActorClass=&LPTT_StaticMeshActor_C, Tag=Active)
   > n2 = For Each Loop(Array:Array of LPTT Static Mesh Actor Object References=@n1) -> (Array Element:LPTT Static Mesh Actor Object Reference)
   Return Node!(execute=@n2.Completed, VertexString:String=Get OutputString() -> (OutputString:String))
   n3 = Get PartType!(self:LPTT Static Mesh Actor Object Reference=@n2) -> (PartType:LPTT_ImportPart_PathGen Enum=NewEnumerator0)
   n4 = "Equal, Case Insensitive (String)"(A=Enum to String!(Enumerator:LPTT_ImportPart_PathGen Enum=@n3) -> (ReturnValue:String), B=Backpack)
   n5 = Get VertexCount!(self:LPTT Static Mesh Actor Object Reference=@n2) -> (VertexCount:int=0)
   n6 = "integer > integer"(A=@n5, B=5000)
   n7 = "Equal, Case Insensitive (String)"(A=Enum to String(Enumerator=@n3), B=Cape)
   n8 = "integer > integer"(A=@n5, B=5000)
   n9 = "Equal, Case Insensitive (String)"(A=Enum to String(Enumerator=@n3), B=Shoe)
   n10 = "integer > integer"(A=@n5, B=4400)
   n11 = "Equal, Case Insensitive (String)"(A=Enum to String(Enumerator=@n3), B=Pet)
   n12 = "integer > integer"(A=@n5, B=5000)
   n13 = "Equal, Case Insensitive (String)"(A=Enum to String(Enumerator=@n3), B=Body)
   n14 = "integer > integer"(A=@n5, B=16000)
   n15 = "Equal, Case Insensitive (String)"(A=Enum to String(Enumerator=@n3), B=Head)
   n16 = "integer > integer"(A=@n5, B=3800)
   n17 = "Equal, Case Insensitive (String)"(A=Enum to String(Enumerator=@n3), B=FaceAcc)
   n18 = "integer > integer"(A=@n5, B=6200)
   > n19 = Branch(Condition=OR Boolean!(
     A=AND Boolean(A=@n13, B=@n14)
     B=AND Boolean(A=@n15, B=@n16)
     C:bool=AND Boolean(A=@n17, B=@n18)
     D:bool=AND Boolean(A=@n4, B=@n6)
     E:bool=AND Boolean(A=@n7, B=@n8)
     F:bool=AND Boolean(A=@n9, B=@n10)
     G:bool=AND Boolean(A=@n11, B=@n12)
    ) -> (ReturnValue))
   n20 = "To String (Integer)"(InInt=@n5)
   n21 = Append!(A="<Error>", B=@n20, C:String="</>") -> (ReturnValue)
   n22 = Get DisplayName!(self:LPTT Static Mesh Actor Object Reference=@n2) -> (DisplayName:String)
   > Set OutputString(OutputString:String=Append!(A=Get OutputString(), B=Append!(A=@n22, B=" : ", C:String=@n21, D:String="\r\n") -> (ReturnValue)) -> (ReturnValue)) -> (Output_Get:String)
   Set OutputString(
    execute=@n19.else
    OutputString=Append!(A=Get OutputString(), B=Append!(A=@n22, B=" : ", C:String=@n20, D:String="\r\n") -> (ReturnValue)) -> (ReturnValue)
   )
  ]}
 Configure_TexturePreviews {nodes [
   n1 = Configure_TexturePreviews!() -> (then, Body:Texture Object Reference, FaceAcc:Texture Object Reference, Head:Texture Object Reference, Hair:Texture Object Reference) {is=title K2Node_FunctionEntry}
   > n2 = Sequence!(execute) -> (then_0, then_1, then_2:Exec, then_3:Exec)
   > n3 = Is Valid(InputObject=@n1)
   n4 = Get BodyTexture() -> (BodyTexture:Image Object Reference)
   SetVisibility(execute=@n3.Is Not Valid, self=@n4, InVisibility=Hidden)
   SetBrushResourceObject!(execute=@n3.Is Valid, self:Image Object Reference=@n4, ResourceObject:Object Reference=@n1) -> (then)
   > SetVisibility(self=@n4, InVisibility=Visible)
   n5 = Is Valid(exec=@n2.then_1, InputObject=@n1.FaceAcc)
   n6 = Get FaceAcc() -> (FaceAcc:Image Object Reference)
   SetVisibility(execute=@n5.Is Not Valid, self=@n6, InVisibility=Hidden)
   SetBrushResourceObject(execute=@n5.Is Valid, self=@n6, ResourceObject=@n1.FaceAcc)
   n7 = Is Valid(exec=@n2.then_2, InputObject=@n1.Head)
   n8 = Get Head() -> (Head:Image Object Reference)
   SetVisibility(execute=@n7.Is Not Valid, self=@n8, InVisibility=Hidden)
   SetBrushResourceObject(execute=@n7.Is Valid, self=@n8, ResourceObject=@n1.Head)
   n9 = Is Valid(exec=@n2.then_3, InputObject=@n1.Hair)
   n10 = Get Hair() -> (Hair:Image Object Reference)
   SetVisibility(execute=@n9.Is Not Valid, self=@n10, InVisibility=Hidden)
   SetBrushResourceObject(execute=@n9.Is Valid, self=@n10, ResourceObject=@n1.Hair)
   SetVisibility(self=@n10, InVisibility=Visible)
   SetVisibility(self=@n6, InVisibility=Visible)
   SetVisibility(self=@n8, InVisibility=Visible)
  ]}
 Configure_TexturePreviewsRenderTarget {nodes [
   n1 = Configure_TexturePreviewsRenderTarget!() -> (then, Body:Texture Object Reference, FaceAcc:Texture Object Reference, Head:Texture Object Reference, Hair:Texture Object Reference) {is=title K2Node_FunctionEntry}
   > n2 = Sequence()
   > n3 = Is Valid(InputObject=@n1)
   n4 = Get BodyTexture()
   SetVisibility(execute=@n3.Is Not Valid, self=@n4, InVisibility=Hidden)
   SetBrushResourceObject(execute=@n3.Is Valid, self=@n4, ResourceObject=@n1)
   > SetVisibility(self=@n4, InVisibility=Visible)
   n5 = Is Valid(exec=@n2.then_1, InputObject=@n1.FaceAcc)
   n6 = Get FaceAcc()
   SetVisibility(execute=@n5.Is Not Valid, self=@n6, InVisibility=Hidden)
   SetBrushResourceObject(execute=@n5.Is Valid, self=@n6, ResourceObject=@n1.FaceAcc)
   > SetVisibility(self=@n6, InVisibility=Visible)
   n7 = Is Valid(exec=@n2.then_2, InputObject=@n1.Head)
   n8 = Get Head()
   SetVisibility(execute=@n7.Is Not Valid, self=@n8, InVisibility=Hidden)
   SetBrushResourceObject(execute=@n7.Is Valid, self=@n8, ResourceObject=@n1.Head)
   > SetVisibility(self=@n8, InVisibility=Visible)
   n9 = Is Valid(exec=@n2.then_3, InputObject=@n1.Hair)
   n10 = Get Hair()
   SetVisibility(execute=@n9.Is Not Valid, self=@n10, InVisibility=Hidden)
   SetBrushResourceObject(execute=@n9.Is Valid, self=@n10, ResourceObject=@n1.Hair)
   > SetVisibility(self=@n10, InVisibility=Visible)
  ]}
 CountValidTags {nodes [
   n1 = CountValidTags!() -> (then, ActiveActors:Array of LPTT Static Mesh Actor Object References) {is=title K2Node_FunctionEntry}
   > n2 = For Each Loop(Array=@n1)
   n3 = Get ActorTags() -> (ActorTags:Array of Names)
   Return Node!(execute=@n2.Completed, ActorTags1:Array of Names=@n3)
   > n4 = For Each Loop(Array:Array of Names=Get Tags!(self:Actor Object Reference=@n2) -> (Tags:Array of Names)) -> (Array Element:Name)
   > n5 = Branch(Condition="Equal (Name)"(A=@n4, B=Active))
   Add Unique(execute=@n5.else, TargetArray:Array of Names=@n3, NewItem:"Name (by ref)"=@n4)
  ]}
 EventGraph {nodes [
   Event PreConstruct()
   > Set Text(Text="To Text (String)"(InString=GetPlatformUserName!() -> (ReturnValue:String)), self=Get Name() -> (Name:Text Object Reference))
   > Set Text(
    Text="As Date (from UTC)"!(
     InDateTime:"Date Time Structure (by ref)"=UTC Now!() -> (ReturnValue:Date Time Structure)
     InTimeZone:String=EDT
     InDateStyle:EDateTimeStyle Enum=Default
    ) -> (ReturnValue:Text)
    self=Get Date() -> (Date:Text Object Reference)
   )
   n1 = GetEngineVersion!() -> (ReturnValue:String)
   > "SetText (Text)"(self=Get Version() -> (Version:Text Object Reference), InText="To Text (String)"(InString=@n1))
   Event Construct()
   > n2 = CompileAllVertexCounts!(execute, self) -> (then, VertexString:String) {is=title K2Node_CallFunction}
   > SetText(self=Get RichTextBlock_355() -> (RichTextBlock_355:Rich Text Block Object Reference), InText="To Text (String)"(InString=@n2))
   > CollectTexturePreviews() {is=title K2Node_CallFunction}
   Split(SourceString=Split(SourceString=@n1, InStr=+++, SearchDir=FromStart), InStr=-, SearchDir=FromStart)
   Event Tick()
  ]}
 HideTextures {nodes [
   n1 = HideTextures!() -> (then, Visible?:bool=~) {is=title K2Node_FunctionEntry}
   > n2 = Branch(Condition=@n1)
   n3 = Get BodyTexture()
   n4 = Get Head()
   n5 = Get FaceAcc()
   n6 = Get Hair()
   > SetOpacity!(execute, self:Image Object Reference=[@n3, @n4, @n5, @n6], InOpacity:float=1) -> (then)
   SetOpacity(execute=@n2.else, self=[@n3, @n4, @n5, @n6], InOpacity=0)
   n7 = Get HeadTextureControl() -> (HeadTextureControl:Vertical Box Object Reference)
   n8 = Get HairTextureControl() -> (HairTextureControl:Vertical Box Object Reference)
   n9 = Get FaceAccTextureControl() -> (FaceAccTextureControl:Vertical Box Object Reference)
   n10 = Get BodyTextureControl() -> (BodyTextureControl:Vertical Box Object Reference)
   SetVisibility(self=[@n7, @n8, @n9, @n10], InVisibility=Collapsed)
   SetVisibility(self=[@n7, @n8, @n9, @n10], InVisibility=Visible)
  ]}
 SetPassText {nodes [
   n1 = SetPassText!() -> (then, Text:Text) {is=title K2Node_FunctionEntry}
   > Set Text(Text=@n1, self=Get Pass() -> (Pass:Text Object Reference))
  ]}
 SwitchTextureToUVIslands {nodes [
   SwitchTextureToUVIslands() {is=title K2Node_FunctionEntry}
   > Configure_TexturePreviews!(
    execute
    self
    Body:Texture Object Reference=Get Islands_Body() -> (Islands_Body:Texture Object Reference)
    FaceAcc:Texture Object Reference=Get Islands_FaceAcc() -> (Islands_FaceAcc:Texture Object Reference)
    Head:Texture Object Reference=Get Islands_Head() -> (Islands_Head:Texture Object Reference)
    Hair:Texture Object Reference=Get Islands_Hair() -> (Islands_Hair:Texture Object Reference)
   ) -> (then) {is=title K2Node_CallFunction}
  ]}
 SwitchTextureToWireFrame {nodes [
   SwitchTextureToWireFrame() {is=title K2Node_FunctionEntry}
   > Configure_TexturePreviewsRenderTarget!(
    execute
    self
    Body:Texture Object Reference=Get WireFrame_Body() -> (WireFrame_Body:Texture Object Reference)
    FaceAcc:Texture Object Reference=Get WireFrame_FaceAcc() -> (WireFrame_FaceAcc:Texture Object Reference)
    Head:Texture Object Reference=Get WireFrame_Head() -> (WireFrame_Head:Texture Object Reference)
    Hair:Texture Object Reference=Get WireFrame_Hair() -> (WireFrame_Hair:Texture Object Reference)
   ) -> (then) {is=title K2Node_CallFunction}
  ]}
]
variables [
 Islands_Body {type=Texture, default=null}
 Islands_FaceAcc {type=Texture, default=null}
 Islands_Hair {type=Texture, default=null}
 Islands_Head {type=Texture, default=null}
 WireFrame_Body {type=Texture, default=null}
 WireFrame_FaceAcc {type=Texture, default=null}
 WireFrame_Hair {type=Texture, default=null}
 WireFrame_Head {type=Texture, default=null}
]
widgets {
 info {inheritedWidgetCount=0, parentClass=&UserWidget, rootWidgetClass=&CanvasPanel, widgetCount=27}
 widgets [
  CanvasPanel_23 {widgetClassPath=&CanvasPanel, widget=CanvasPanel_23}
  Border_8 {
   widgetClassPath=&Border
   parent=CanvasPanel_23
   widget=Border_8
   properties {brushColor=<rgba 0.2 0.2 0.2 1>}
   slot {
    class=CanvasPanelSlot
    properties {bAutoSize=true, layoutData {alignment=<0 0>, anchors {maximum=<0 0>, minimum=<0 0>}, offsets=<ltrb 0 0 800 50>}}
   }
  }
  Name {
   widgetClassPath=&TextBlock
   bIsVariable=true
   parent=Border_8
   widget=Name
   properties {
    text=Name
    font {
     bForceMonospaced=false
     bMaterialIsStencil=false
     fontMaterial=null
     fontObject=&Roboto
     letterSpacing=0
     monospacedWidth=1
     size=30
     skewAmount=0
     typefaceFontName=Bold
     outlineSettings {
      bApplyOutlineToDropShadows=false
      bMiteredCorners=false
      bSeparateFillAlpha=false
      outlineMaterial=null
      outlineSize=0
      outlineColor=<rgba 0 0 0 1>
     }
    }
   }
   slot {class=BorderSlot}
  }
  Border_3 {
   widgetClassPath=&Border
   parent=CanvasPanel_23
   widget=Border_3
   properties {brushColor=<rgba 0.2 0.2 0.2 1>}
   slot {
    class=CanvasPanelSlot
    properties {bAutoSize=true, layoutData {alignment=<0 0>, anchors {maximum=<1 0>, minimum=<1 0>}, offsets=<ltrb -446.4801 0 800 50>}}
   }
  }
  Pass {
   widgetClassPath=&TextBlock
   bIsVariable=true
   parent=Border_3
   widget=Pass
   properties {
    justification=Right
    text=Pass
    font {
     bForceMonospaced=false
     bMaterialIsStencil=false
     fontMaterial=null
     fontObject=&Roboto
     letterSpacing=0
     monospacedWidth=1
     size=30
     skewAmount=0
     typefaceFontName=Bold
     outlineSettings {
      bApplyOutlineToDropShadows=false
      bMiteredCorners=false
      bSeparateFillAlpha=false
      outlineMaterial=null
      outlineSize=0
      outlineColor=<rgba 0 0 0 1>
     }
    }
   }
   slot {class=BorderSlot}
  }
  Border {
   widgetClassPath=&Border
   parent=CanvasPanel_23
   widget=Border
   properties {brushColor=<rgba 0.2 0.2 0.2 1>}
   slot {
    class=CanvasPanelSlot
    properties {bAutoSize=true, layoutData {alignment=<0 0>, anchors {maximum=<0 0>, minimum=<0 0>}, offsets=<ltrb 0 52 800 50>}}
   }
  }
  Date {
   widgetClassPath=&TextBlock
   bIsVariable=true
   parent=Border
   widget=Date
   properties {
    text=Date
    font {
     bForceMonospaced=false
     bMaterialIsStencil=false
     fontMaterial=null
     fontObject=&Roboto
     letterSpacing=0
     monospacedWidth=1
     size=30
     skewAmount=0
     typefaceFontName=Bold
     outlineSettings {
      bApplyOutlineToDropShadows=false
      bMiteredCorners=false
      bSeparateFillAlpha=false
      outlineMaterial=null
      outlineSize=0
      outlineColor=<rgba 0 0 0 1>
     }
    }
   }
   slot {class=BorderSlot}
  }
  Border_1 {
   widgetClassPath=&Border
   parent=CanvasPanel_23
   widget=Border_1
   properties {brushColor=<rgba 0.2 0.2 0.2 1>}
   slot {
    class=CanvasPanelSlot
    properties {bAutoSize=true, layoutData {alignment=<0 0>, anchors {maximum=<0 0>, minimum=<0 0>}, offsets=<ltrb 0 104 800 50>}}
   }
  }
  Version {
   widgetClassPath=&TextBlock
   bIsVariable=true
   parent=Border_1
   widget=Version
   properties {
    text=Version
    font {
     bForceMonospaced=false
     bMaterialIsStencil=false
     fontMaterial=null
     fontObject=&Roboto
     letterSpacing=0
     monospacedWidth=1
     size=30
     skewAmount=0
     typefaceFontName=Bold
     outlineSettings {
      bApplyOutlineToDropShadows=false
      bMiteredCorners=false
      bSeparateFillAlpha=false
      outlineMaterial=null
      outlineSize=0
      outlineColor=<rgba 0 0 0 1>
     }
    }
   }
   slot {class=BorderSlot}
  }
  Border_2 {
   widgetClassPath=&Border
   parent=CanvasPanel_23
   widget=Border_2
   properties {horizontalAlignment=HAlign_Left, brushColor=<rgba 0.2 0.2 0.2 1>}
   slot {
    class=CanvasPanelSlot
    properties {bAutoSize=true, layoutData {alignment=<0 0>, anchors {maximum=<0 0>, minimum=<0 0>}, offsets=<ltrb 750 0 800 50>}}
   }
  }
  RichTextBlock_355 {
   widgetClassPath=/Script/UMG.RichTextBlock
   bIsVariable=true
   parent=Border_2
   widget=RichTextBlock_355
   properties {text=VertexCounts, textStyleSet=&/DT_LPTT_RichText}
   slot {class=BorderSlot, properties {horizontalAlignment=HAlign_Left}}
  }
  Overlay_0 {
   widgetClassPath=&Overlay
   parent=CanvasPanel_23
   widget=Overlay_0
   properties {bOverride_Cursor=true}
   slot {class=CanvasPanelSlot, properties {layoutData {alignment=<0 0>, anchors {maximum=<1 1>, minimum=<0 0>}, offsets=<ltrb 0 0 5.1051 3.0331>}}}
  }
  VerticalBox_2 {
   widgetClassPath=&VerticalBox
   parent=Overlay_0
   widget=VerticalBox_2
   slot {class=OverlaySlot, properties {horizontalAlignment=HAlign_Fill, verticalAlignment=VAlign_Fill}}
  }
  HorizontalBox_0 {
   widgetClassPath=&HorizontalBox
   parent=VerticalBox_2
   widget=HorizontalBox_0
   slot {class=VerticalBoxSlot, properties {size {sizeRule=Fill, value=1}}}
  }
  VerticalBox_0 {
   widgetClassPath=&VerticalBox
   parent=HorizontalBox_0
   widget=VerticalBox_0
   slot {class=HorizontalBoxSlot, properties {size {sizeRule=Fill, value=1}}}
  }
  Spacer_346 {
   widgetClassPath=&Spacer
   parent=VerticalBox_0
   widget=Spacer_346
   slot {class=VerticalBoxSlot, properties {padding=<ltrb 0 500 0 0>}}
  }
  HeadTextureControl {
   widgetClassPath=&VerticalBox
   bIsVariable=true
   parent=VerticalBox_0
   widget=HeadTextureControl
   slot {class=VerticalBoxSlot, properties {size {sizeRule=Fill, value=1}}}
  }
  BodyTextureControl {
   widgetClassPath=&VerticalBox
   bIsVariable=true
   parent=VerticalBox_0
   widget=BodyTextureControl
   slot {class=VerticalBoxSlot, properties {size {sizeRule=Fill, value=1}}}
  }
  Spacer_201 {
   widgetClassPath=&Spacer
   parent=HorizontalBox_0
   widget=Spacer_201
   slot {class=HorizontalBoxSlot, properties {padding=<ltrb 300 0 300 0>, size {sizeRule=Fill, value=1}}}
  }
  VerticalBox_1 {
   widgetClassPath=&VerticalBox
   parent=HorizontalBox_0
   widget=VerticalBox_1
   slot {class=HorizontalBoxSlot, properties {size {sizeRule=Fill, value=1}}}
  }
  Spacer {widgetClassPath=&Spacer, parent=VerticalBox_1, widget=Spacer, slot {class=VerticalBoxSlot, properties {padding=<ltrb 0 500 0 0>}}}
  HairTextureControl {
   widgetClassPath=&VerticalBox
   bIsVariable=true
   parent=VerticalBox_1
   widget=HairTextureControl
   slot {class=VerticalBoxSlot, properties {size {sizeRule=Fill, value=1}}}
  }
  FaceAccTextureControl {
   widgetClassPath=&VerticalBox
   bIsVariable=true
   parent=VerticalBox_1
   widget=FaceAccTextureControl
   slot {class=VerticalBoxSlot, properties {size {sizeRule=Fill, value=1}}}
  }
  BodyTexture {
   widgetClassPath=&Image
   bIsVariable=true
   parent=CanvasPanel_23
   widget=BodyTexture
   properties {brush {
     bIsDynamicallyLoaded=false
     drawAs=Image
     imageType=FullColor
     mirroring=NoMirror
     resourceName=None
     resourceObject=null
     tiling=NoTile
     imageSize=<1500 1500>
     margin=<ltrb 0 0 0 0>
     outlineSettings {
      bUseBrushTransparency=false
      roundingType=HalfHeightRadius
      width=0
      color {colorUseRule=UseColor_Specified, specifiedColor=<rgba 0 0 0 0>}
      cornerRadii=<0 0 0 0>
     }
     tintColor {colorUseRule=UseColor_Specified, specifiedColor=<rgba 1 1 1 1>}
     uVRegion {bIsValid=false, max=<0 0>, min=<0 0>}
    }}
   slot {
    class=CanvasPanelSlot
    properties {bAutoSize=true, layoutData {alignment=<0 0>, anchors {maximum=<0 1>, minimum=<0 1>}, offsets=<ltrb 4 -1508 1000 1000>}}
   }
  }
  Head {
   widgetClassPath=&Image
   bIsVariable=true
   parent=CanvasPanel_23
   widget=Head
   properties {brush {
     bIsDynamicallyLoaded=false
     drawAs=Image
     imageType=FullColor
     mirroring=NoMirror
     resourceName=None
     resourceObject=null
     tiling=NoTile
     imageSize=<1500 1500>
     margin=<ltrb 0 0 0 0>
     outlineSettings {
      bUseBrushTransparency=false
      roundingType=HalfHeightRadius
      width=0
      color {colorUseRule=UseColor_Specified, specifiedColor=<rgba 0 0 0 0>}
      cornerRadii=<0 0 0 0>
     }
     tintColor {colorUseRule=UseColor_Specified, specifiedColor=<rgba 1 1 1 1>}
     uVRegion {bIsValid=false, max=<0 0>, min=<0 0>}
    }}
   slot {
    class=CanvasPanelSlot
    properties {bAutoSize=true, layoutData {alignment=<0 0>, anchors {maximum=<0 1>, minimum=<0 1>}, offsets=<ltrb 4 -3012 1000 1000>}}
   }
  }
  FaceAcc {
   widgetClassPath=&Image
   bIsVariable=true
   parent=CanvasPanel_23
   widget=FaceAcc
   properties {brush {
     bIsDynamicallyLoaded=false
     drawAs=Image
     imageType=FullColor
     mirroring=NoMirror
     resourceName=None
     resourceObject=null
     tiling=NoTile
     imageSize=<1500 1500>
     margin=<ltrb 0 0 0 0>
     outlineSettings {
      bUseBrushTransparency=false
      roundingType=HalfHeightRadius
      width=0
      color {colorUseRule=UseColor_Specified, specifiedColor=<rgba 0 0 0 0>}
      cornerRadii=<0 0 0 0>
     }
     tintColor {colorUseRule=UseColor_Specified, specifiedColor=<rgba 1 1 1 1>}
     uVRegion {bIsValid=false, max=<0 0>, min=<0 0>}
    }}
   slot {
    class=CanvasPanelSlot
    properties {bAutoSize=true, layoutData {alignment=<0 0>, anchors {maximum=<1 1>, minimum=<1 1>}, offsets=<ltrb -1509.4954 -1508 1000 1000>}}
   }
  }
  Hair {
   widgetClassPath=&Image
   bIsVariable=true
   parent=CanvasPanel_23
   widget=Hair
   properties {brush {
     bIsDynamicallyLoaded=false
     drawAs=Image
     imageType=FullColor
     mirroring=NoMirror
     resourceName=None
     resourceObject=null
     tiling=NoTile
     imageSize=<1500 1500>
     margin=<ltrb 0 0 0 0>
     outlineSettings {
      bUseBrushTransparency=false
      roundingType=HalfHeightRadius
      width=0
      color {colorUseRule=UseColor_Specified, specifiedColor=<rgba 0 0 0 0>}
      cornerRadii=<0 0 0 0>
     }
     tintColor {colorUseRule=UseColor_Specified, specifiedColor=<rgba 1 1 1 1>}
     uVRegion {bIsValid=false, max=<0 0>, min=<0 0>}
    }}
   slot {
    class=CanvasPanelSlot
    properties {bAutoSize=true, layoutData {alignment=<0 0>, anchors {maximum=<1 1>, minimum=<1 1>}, offsets=<ltrb -1509.4954 -3012 1000 1000>}}
   }
  }
 ]
}
